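# UI material function that turns a distance field into either a smooth-stepped glow or an anti-aliased shape.
# The glow is an inverted SmoothStep of the field between GlowMin and GlowMax.
# An IsEqual test on those inputs drives a lerp to MF_UI_AntiAliasing's Inner output when they match, rather than using a static switch.
# AA Threshold and AA Smoothing affect only that anti-aliased result.
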MaterialFunction /FortUILibrary/MaterialFunctions/Shapes/Modifiers/MF_UI_SDFGlowOrAA_WithThreshold
graph {
 properties {
  bExposeToLibrary=true
  description="Applies a smooth-stepped falloff (glow) to a distance field. Returns an anti-aliased result if GlowMin and GlowMax values are equal."
  previewMaterialDomain=MD_UI
  libraryCategoriesText [UI_Library]
 }
 nodes [
  n1 = FunctionInput() {
   bUsePreviewValueAsDefault=true
   desc="Determines glow falloff strength. Can not go above GlowMax value or results will invert. When Glow Min equals Glow Max, \r\nthe glow is disabled and replaced by anti-aliasing. AA Threshold becomes active instead."
   description="Determines glow falloff strength. Can not go above GlowMax value or results will invert. When Glow Min equals Glow Max, the glow is disabled and replaced by anti-aliasing. AA Threshold becomes active instead."
   inputName=GlowMin
   inputType=FunctionInput_Scalar
   sortPriority=1
  }
  n2 = FunctionInput() {
   bUsePreviewValueAsDefault=true
   desc="Determines glow falloff strength. Can not go below GlowMin value or results will invert. When Glow Min equals Glow Max, \r\nthe glow is disabled and replaced by anti-aliasing. AA Threshold becomes active instead."
   description="Determines glow falloff strength. Can not go below GlowMin value or results will invert. When Glow Min equals Glow Max, the glow is disabled and replaced by anti-aliasing. AA Threshold becomes active instead."
   inputName=GlowMax
   inputType=FunctionInput_Scalar
   sortPriority=2
  }
  n3 = FunctionInput(Preview=MaterialFunctionCall!(UVs, Radius) {materialFunction=/FortUILibrary/MaterialFunctions/Shapes/SDFs/MF_SDF_Circle}.SDF) {description=The distance field to sample., inputName=DistanceField, inputType=FunctionInput_Scalar}
  n4 = SmoothStep(Max=Subtract(A=@n2, B=@n1), Value=Subtract(A=@n3, B=@n1))
  n5 = FunctionInput() {
   bUsePreviewValueAsDefault=true
   desc="Useful if you need to increase the thickness of your SDF shape without using a Scale.\r\nThis feature is unavailable while the Glow effect is active."
   description=Useful if you need to increase the thickness of your SDF shape without using a Scale.This feature is unavailable while the Glow effect is active.
   inputName=AA Threshold
   inputType=FunctionInput_Scalar
   sortPriority=3
  }
  n6 = FunctionInput() {
   bUsePreviewValueAsDefault=true
   desc="Blurs/softens the resulting edges. Note that adjusting this value will bypass anti-aliasing!\r\nThis feature is unavailable while the Glow effect is active."
   description="Blurs/softens the resulting edges. Note that adjusting this value will bypass anti-aliasing!This feature is unavailable while the Glow effect is active."
   inputName=AA Smoothing
   inputType=FunctionInput_Scalar
   sortPriority=4
  }
  FunctionOutput(None=LinearInterpolate(
    A=OneMinus(None=@n4)
    B=MaterialFunctionCall(DistanceField=@n3, Threshold=@n5, Smoothing=@n6) {materialFunction=&MF_UI_AntiAliasing}.Inner
    Alpha=MaterialFunctionCall!(A=@n1, B=@n2) {materialFunction=/FortUILibrary/MaterialFunctions/Basics/MF_UI_BooleanOperations}.IsEqual
   ) {constA=1, constB=0}) {description=The resulting shape with anti-aliasing or smooth-stepped glow.}
 ]
}
