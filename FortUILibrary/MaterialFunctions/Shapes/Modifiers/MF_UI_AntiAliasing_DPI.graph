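# UI material function that returns an anti-aliased mask from an SDF shape or distance field texture, compensating for UI DPI scale.
# A SmoothStep ramps from zero to Threshold divided by View.ViewportScaleUI, read through a Custom node, so the edge narrows as scale rises.
# Outer outputs the ramp for white-interior textures, and Inner outputs its inverse for negative-interior procedural SDFs.
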MaterialFunction /FortUILibrary/MaterialFunctions/Shapes/Modifiers/MF_UI_AntiAliasing_DPI
graph {
 properties {
  bExposeToLibrary=true
  description=Returns an anti-aliased version of the supplied SDF shape or distance field texture.
  previewMaterialDomain=MD_UI
  libraryCategoriesText [UI_Library]
 }
 nodes [
  n1 = FunctionInput(Preview=Constant() {r=0.02}) {
   bUsePreviewValueAsDefault=true
   description="The field value that anti-aliasing will occur around. Use \"0.0\" to anti-alias around the exact boundary of procedural SDFs. Use \"0.5\" to anti-alias around mid-gray for texture-based DFs."
   inputName=Threshold
   inputType=FunctionInput_Scalar
   sortPriority=1
  }
  n2 = FunctionInput(Preview=MaterialFunctionCall!(UVs, Radius) {materialFunction=/FortUILibrary/MaterialFunctions/Shapes/SDFs/MF_SDF_Circle}.SDF) {
   description="The distance field (eg. SDF shape or greyscale DF texture) to perform anti-aliasing on."
   inputName=DistanceField
   inputType=FunctionInput_Scalar
  }
  n3 = SmoothStep(
   Max=Divide(A=@n1, B=Custom() {code="return View.ViewportScaleUI;", description=ViewportScaleFactor, outputType=CMOT_Float1, inputs []})
   Value=@n2
  )
  FunctionOutput(None=@n3) {description=Works best with texture-based DFs where shape interior is white and exterior is black., outputName=Outer, sortPriority=1}
  FunctionOutput(None=OneMinus(None=@n3)) {
   description=Works best with procedural SDF shapes where negative values indicate shape interior and positive values indicate the exterior.
   outputName=Inner
  }
 ]
}
