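# UI material function that turns a distance field into a glowing shape, or an anti-aliased one when GlowMin equals GlowMax.
# The field is offset by GlowMin, then smooth-stepped over the GlowMax minus GlowMin range and inverted to form the falloff.
# The same offset field also feeds MF_UI_AntiAliasing.
# An equality check from MF_UI_BooleanOperations drives a lerp between the two results instead of a branch.
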MaterialFunction /FortUILibrary/MaterialFunctions/Shapes/Modifiers/MF_UI_SDFGlowOrAA
graph {
 properties {
  bExposeToLibrary=true
  description="Applies a smooth-stepped falloff (glow) to a distance field. Returns an anti-aliased result if GlowMin and GlowMax values are equal."
  previewMaterialDomain=MD_UI
  libraryCategoriesText [UI_Library]
 }
 nodes [
  n1 = FunctionInput() {
   bUsePreviewValueAsDefault=true
   description=Determines glow falloff strength. Can not go below GlowMin value or results will invert.
   inputName=GlowMax
   inputType=FunctionInput_Scalar
   sortPriority=2
  }
  n2 = FunctionInput() {
   bUsePreviewValueAsDefault=true
   description=Determines glow falloff strength. Can not go above GlowMax value or results will invert.
   inputName=GlowMin
   inputType=FunctionInput_Scalar
   sortPriority=1
  }
  n3 = Subtract(
   A=FunctionInput(Preview=MaterialFunctionCall!(UVs, Radius) {materialFunction=/FortUILibrary/MaterialFunctions/Shapes/SDFs/MF_SDF_Circle}.SDF) {description=The distance field to sample., inputName=DistanceField, inputType=FunctionInput_Scalar}
   B=@n2
  )
  n4 = SmoothStep(Max=Subtract(A=@n1, B=@n2), Value=@n3)
  FunctionOutput(None=LinearInterpolate(
    A=OneMinus(None=@n4)
    B=MaterialFunctionCall(DistanceField=@n3) {materialFunction=&MF_UI_AntiAliasing}.Inner
    Alpha=MaterialFunctionCall!(A=@n2, B=@n1) {materialFunction=/FortUILibrary/MaterialFunctions/Basics/MF_UI_BooleanOperations}.IsEqual
   )) {description=The resulting shape with anti-aliasing or smooth-stepped glow.}
 ]
}
