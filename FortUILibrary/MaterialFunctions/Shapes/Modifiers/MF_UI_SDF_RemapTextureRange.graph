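# UI library material function that remaps a 0 to 1 distance field texture mask to a -1 to 1 signed distance field gradient.
# It multiplies the scalar input by two, then subtracts the default one to produce the SDF output.
# No inversion occurs, so white (filled) maps to positive, although the input description states that negative values are inside the shape.
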MaterialFunction /FortUILibrary/MaterialFunctions/Shapes/Modifiers/MF_UI_SDF_RemapTextureRange
graph {
 properties {
  bExposeToLibrary=true
  description=Convert a 0 to 1 distance field texture gradient to a -1 to 1 signed distance field gradient.
  previewMaterialDomain=MD_UI
  libraryCategoriesText [UI_Library]
 }
 nodes [
  n1 = FunctionInput() {
   bUsePreviewValueAsDefault=true
   description="Black (0) represents empty space, white (1) represents filled areas. This mask will be remapped from (0, 1) to (–1, 1) to create a Signed Distance Field (SDF), where negative values are inside the shape and positive values are outside."
   inputName="Texture mask (0-1 range)"
   inputType=FunctionInput_Scalar
  }
  FunctionOutput(None=Subtract(A=Multiply(A=@n1) {constB=2, desc="range = 0 to 2"}) {desc="range = -1,1\r\nSigned distance field"}) {description="The resulting texture remapped to -1,1.", outputName=SDF}
 ]
}
