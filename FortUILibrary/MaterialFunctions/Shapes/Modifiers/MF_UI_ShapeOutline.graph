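# UI material function that turns a distance field into an outline, returning both an anti-aliased mask and the outline SDF.
# It shifts the field by Threshold and by Alignment times Thickness, then takes the absolute value minus Thickness to form the band.
# A Ceil/Saturate gate on absolute Thickness flattens the SDF to zero at zero thickness.
# MF_UI_AntiAliasing applies Smoothing to the Outline output only, leaving SDFOutline raw for further operations.
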MaterialFunction /FortUILibrary/MaterialFunctions/Shapes/Modifiers/MF_UI_ShapeOutline
graph {
 properties {
  bExposeToLibrary=true
  description=Returns an anti-aliased outline/stroke or outline SDF for a given distance field.
  previewMaterialDomain=MD_UI
  libraryCategoriesText [UI_Library]
 }
 nodes [
  n1 = FunctionInput(Preview=MaterialFunctionCall!(UVs, Radius) {materialFunction=/FortUILibrary/MaterialFunctions/Shapes/SDFs/MF_SDF_Circle}.SDF) {
   description="The distance field (eg. SDF shape or greyscale DF texture) to generate an outline from."
   inputName=DistanceField
   inputType=FunctionInput_Scalar
  }
  n2 = FunctionInput() {
   bUsePreviewValueAsDefault=true
   description="The field value that the outline will appear around. Use \"0.0\" to generate an outline around the exact boundary of procedural SDFs. Use \"0.5\" to generate an outline around mid-gray for texture-based DFs."
   inputName=Threshold
   inputType=FunctionInput_Scalar
   sortPriority=1
  }
  n3 = FunctionInput() {
   bUsePreviewValueAsDefault=true
   description="The alignment of the outline (0.0 = center, -1.0 = inner, 1.0 = outer)."
   inputName=Alignment
   inputType=FunctionInput_Scalar
   sortPriority=3
  }
  n4 = FunctionInput() {
   bUsePreviewValueAsDefault=true
   description=The thickness of the outline.
   inputName=Thickness
   inputType=FunctionInput_Scalar
   sortPriority=2
   previewValue=<0.05 0 0 1>
  }
  n5 = Subtract(A=Subtract(A=@n1, B=@n2), B=Multiply(A=@n3, B=@n4))
  n6 = Abs(None=@n4)
  n7 = LinearInterpolate(B=Subtract(A=Abs(None=@n5), B=@n4) {constB=0.05}, Alpha=Ceil(None=Saturate(None=@n6)))
  n8 = FunctionInput() {
   bUsePreviewValueAsDefault=true
   description="Softens the resulting edges of the Outline (does not affect SDFOutline)."
   inputName=Smoothing
   inputType=FunctionInput_Scalar
   sortPriority=4
  }
  FunctionOutput(None=MaterialFunctionCall(DistanceField=@n7, Smoothing=@n8) {materialFunction=&MF_UI_AntiAliasing}.Inner) {description=The anti-aliased outline., outputName=Outline}
  FunctionOutput(None=@n7) {description=The outline SDF for any further operations., outputName=SDFOutline, sortPriority=1}
 ]
}
